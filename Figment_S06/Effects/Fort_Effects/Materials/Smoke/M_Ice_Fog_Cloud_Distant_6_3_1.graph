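# Translucent unlit material for distant ice fog clouds, fading near water height and by depth.
# Vertex offset expands mesh UVs into camera-facing cards, with sine drift and a Reveal parameter driving spread, offset and opacity.
# Cloud normals blend toward a sphere normal by a Distance parameter, feeding Fresnel rim and sun/moon highlight.
# Time-of-day tints, panning detail and a spherical edge-color mask shape color, with cheaper quality and feature-level fallbacks.
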Material /Figment_S06/Effects/Fort_Effects/Materials/Smoke/M_Ice_Fog_Cloud_Distant_6_3_1
graph {
 properties {blendMode=BLEND_Translucent, pixelDepthOffsetMode=PDOM_Legacy, shadingModel=MSM_Unlit, twoSided=true, usedShadingModels=Unlit}
 nodes [
  n1 = TextureCoordinate() {coordinateIndex=2}
  n2 = ScalarParameter() {defaultValue=1, parameterName=Distance}
  n3 = LinearInterpolate(
   A=TextureSample(UVs=@n1) {samplerType=SAMPLERTYPE_Normal, texture=/Figment_S06/Effects/Fort_Effects/Textures/Smoke/T_Cloud_N}.RGB
   B=Multiply(
    A=TextureSample() {samplerType=SAMPLERTYPE_Normal, texture=/Game/Environments/World/Backgrounds/Universal/TechContent/Textures/T_Spherenormal_01}.RGB
    B=VectorParameter() {parameterName=SoftNormalMult, defaultValue=<rgba 1.2 1.2 0.8 1>}.RGB
   )
   Alpha=Clamp(None=Multiply(A=@n2) {constB=2.7}, Min=ScalarParameter() {defaultValue=0.25, parameterName=DistanceMin}) {maxDefault=0.8}
  )
  n4 = Fresnel(ExponentIn=ScalarParameter() {defaultValue=2.5, parameterName=FresnelExp}, Normal=Transform(None=@n3)) {baseReflectFraction=0, exponent=2}
  Add(A=@n4) {constB=0.1}
  n5 = VertexColor()
  n6 = ScalarParameter() {defaultValue=0.3, parameterName=XSpreadSineOffset}
  n7 = TextureCoordinate() {coordinateIndex=1}
  n8 = ComponentMask(None=@n7) {r=true}
  n9 = Multiply(A=Sine(None=Add(A=@n6, B=@n8)) {period=0.33}) {constB=0.5}
  n10 = ComponentMask(None=@n7) {g=true, r=true}
  n11 = Multiply(A=AppendVector(A=Multiply(A=@n5.R, B=@n9), B=Subtract(A=@n10) {constB=0.5})) {constB=-1}
  n12 = Constant3Vector() {constant=<rgba 0 0.2 0 0>}
  n13 = ScalarParameter() {parameterName=Reveal}
  n14 = Multiply(
   A=Multiply(
    A=Transform(None=@n11) {transformSourceType=TRANSFORMSOURCE_Local}
    B=VectorParameter() {parameterName=Spread, defaultValue=<rgba 120 480 550 0>}.RGB
   )
   B=Add(A=Multiply(A=@n12, B=@n13))
  )
  n15 = Multiply(A=VectorParameter() {parameterName=RevealOffset, defaultValue=<rgba 0 14000 -9000 0>}.RGB, B=@n13)
  n16 = MaterialFunctionCall(Float2=Subtract(A=TextureCoordinate()) {constB=0.5}) {materialFunction=&BreakOutFloat2Components}
  n17 = Subtract(A=ObjectPositionWS(), B=CameraPositionWS())
  n18 = CrossProduct(A=@n17, B=Constant3Vector() {constant=<rgba 0 0 1 0>})
  n19 = Multiply(A=@n16.R, B=Normalize(VectorInput=@n18))
  n20 = ScalarParameter() {defaultValue=115000, parameterName=Scale}
  n21 = SphereMask(A=TextureCoordinate() {coordinateIndex=1}, B=Constant() {r=0.5}) {attenuationRadius=0.9, hardnessPercent=0}
  n22 = Constant() {r=1}
  n23 = ScalarParameter() {defaultValue=0.5, parameterName=HeightScaleMult}
  n24 = MaterialFunctionCall(Float2=Multiply(A=Multiply(A=@n20, B=@n21), B=AppendVector(A=@n22, B=@n23))) {materialFunction=&BreakOutFloat2Components}
  n25 = Multiply(A=@n16.G, B=Normalize(VectorInput=CrossProduct(A=@n17, B=@n18)))
  n26 = Subtract(A=Add(A=Multiply(A=@n19, B=@n24.R), B=Multiply(A=@n25, B=@n24.G)), B=WorldPosition().XYZ)
  n27 = ObjectPositionWS()
  n28 = Constant()
  n29 = MaterialFunctionCall(Float2=Add(A=ScalarParameter() {defaultValue=0.26, parameterName=DriftSineOffset}, B=TextureCoordinate() {coordinateIndex=1})) {materialFunction=&BreakOutFloat2Components}
  n30 = Sine(None=@n29.R) {period=0.33}
  n31 = Sine(None=@n29.G) {period=0.41}
  n32 = Multiply(
   A=Multiply(A=Time(), B=ScalarParameter() {defaultValue=0.5, parameterName=DriftSpeed})
   B=AppendVector(A=AppendVector(A=@n28, B=@n30), B=Multiply(A=@n31) {constB=0.5})
  )
  Add(
   A=Add(A=Add(A=@n14, B=@n15), B=Add(A=@n26, B=@n27))
   B=Multiply(A=Sine(None=@n32) {period=5}, B=ScalarParameter() {defaultValue=2500, parameterName=DriftDistance})
  )
  n33 = ComponentMask(None=Subtract(A=WorldPosition().XYZ, B=ScalarParameter() {defaultValue=-5000, parameterName=WaterHeight})) {b=true}
  n34 = Multiply(A=@n2) {constB=2}
  n35 = Multiply(A=OneMinus(None=Clamp(None=@n34)), B=ScalarParameter() {defaultValue=20000, parameterName=HeightFadeGradientSize})
  n36 = LinearInterpolate(A=ScalarParameter() {defaultValue=1.4, parameterName=OpacityMult}, Alpha=ScalarParameter() {parameterName=Reveal})
  n37 = TextureSample(UVs=@n1) {texture=/Figment_S06/Effects/Fort_Effects/Textures/Smoke/T_Cloud_M}
  n38 = DepthFade(FadeDistance=ScalarParameter() {defaultValue=10000, parameterName=DepthFadeDist})
  n39 = Constant() {r=1}
  Multiply(
   A=Clamp(None=Divide(A=@n33, B=@n35))
   B=Multiply(A=Multiply(A=@n37.G, B=@n36), B=FeatureLevelSwitch(Default=@n38, ES2=@n39, ES3_1=@n39, SM4_REMOVED=@n38))
  )
  n40 = Add(A=ObjectPositionWS(), B=VectorParameter() {parameterName=SphereMaskOffset, defaultValue=<rgba 0 0 125000 1>}.RGB)
  n41 = WorldPosition()
  n42 = ScalarParameter() {defaultValue=148000, parameterName=SphereMaskRadius}
  n43 = TextureCoordinate() {coordinateIndex=2, uTiling=1.5, vTiling=1.5}
  n44 = Multiply(A=Time(), B=ScalarParameter() {defaultValue=0.03, parameterName=PanSpeed})
  n45 = LinearInterpolate(
   A=ScalarParameter() {defaultValue=0.6, parameterName=DetailTexMin}
   Alpha=TextureSample(UVs=Panner(Coordinate=@n43, Time=@n44) {speedX=0.05, speedY=-1}) {texture=/Game/VisualThreatManager/StormVisuals/Textures/2ndInfectedMapLinear}.G
  ) {constB=1.1}
  n46 = Constant() {r=1}
  n47 = FeatureLevelSwitch(Default=@n45, ES2=@n46, ES3_1=@n46, SM4_REMOVED=@n45)
  n48 = VectorParameter() {parameterName=ShadowMorning, defaultValue=<rgba 0.9 0.5487 0.4078 0>}
  n49 = VectorParameter() {parameterName=ShadowDay, defaultValue=<rgba 0.4531 0.5749 1 0>}
  n50 = VectorParameter() {parameterName=ShadowEvening, defaultValue=<rgba 0.4 0.1457 0.1537 1>}
  n51 = VectorParameter() {parameterName=ShadowNight, defaultValue=<rgba 0.3649 0.3478 0.7 0>}
  n52 = MaterialFunctionCall(
   Color_Evening=VectorParameter() {parameterName=EveningTint, defaultValue=<rgba 1.5 0.7295 0.3516 1>}.RGB
   Color_Night=VectorParameter() {parameterName=NightTint, defaultValue=<rgba 0.3311 0.2457 0.7 1>}.RGB
  ) {materialFunction=&Color_Time_of_Day}
  n53 = VectorParameter() {parameterName=FresnelDay, defaultValue=<rgba 0.1669 0.3057 0.4 0>}
  n54 = MaterialFunctionCall(
   Color_Morning=Multiply(A=VectorParameter() {parameterName=FresnelMorning, defaultValue=<rgba 1 0.4768 0.2448 0>}.RGB) {constB=2}
   Color_Day=@n53.RGB
   Color_Evening=Constant()
   Color_Night=Multiply(A=@n53.RGB) {constB=1.25}
  ) {materialFunction=&Color_Time_of_Day}
  n55 = Multiply(A=@n37.B, B=Multiply(A=@n4, B=ScalarParameter() {defaultValue=12, parameterName=FresnelMult}))
  n56 = Add(
   A=Multiply(
    A=QualitySwitch(Default=@n47, Low=@n46)
    B=MaterialFunctionCall(Color_Morning=@n48.RGB, Color_Day=@n49.RGB, Color_Evening=@n50.RGB, Color_Night=@n51.RGB) {materialFunction=&Color_Time_of_Day}.Result
   )
   B=Multiply(A=@n52.Result, B=Multiply(A=@n54.Result, B=@n55))
  ) {constB=0}
  n57 = CollectionParameter() {
   collection=&FortniteMaterialParameters
   parameterName=FortniteActiveTimeOfDay
  }
  n58 = CollectionParameter() {
   collection=&FortniteMaterialParameters
   parameterName=SunAndMoonModelDirectionalVector
  }
  n59 = Add(A=ComponentMask(None=@n58) {b=true, g=true, r=true}, B=VectorParameter() {parameterName=SunVectorOffset, defaultValue=<rgba 0.4 0 0 1>}.RGB)
  n60 = Multiply(A=VectorParameter() {parameterName=NormalTransform, defaultValue=<rgba -1 1 1 0>}.RGB, B=@n3)
  n61 = Clamp(None=DotProduct(A=Normalize(VectorInput=@n59), B=Transform(None=@n60)))
  n62 = Cosine(None=Power(
    Base=CollectionParameter() {collection=&FortniteMaterialParameters, parameterName=TimeOfDay_0-1}
    Exp=ScalarParameter() {defaultValue=2, parameterName=TODPower}
   ))
  n63 = QualitySwitch(
   Default=LinearInterpolate(Alpha=ConstantBiasScale(None=@n62)) {constA=6, constB=1.8}
   Low=ScalarParameter() {defaultValue=2, parameterName=Spec Highlight Power}
  )
  n64 = Multiply(
   A=LinearInterpolate(Alpha=ComponentMask(None=@n57) {a=true}) {constA=1, constB=0.5}
   B=Multiply(
    A=VectorParameter() {parameterName=HighlightColor, defaultValue=<rgba 2.0312 2.4144 2.5 0>}.RGB
    B=Power(Base=@n61, Exp=@n63) {constExponent=20}
   ) {constB=0}
  )
  Multiply(
   A=LinearInterpolate(
    A=VectorParameter() {parameterName=EdgeColor, defaultValue=<rgba 0.4675 0.5275 0.55 0>}.RGB
    Alpha=SphereMask(A=@n40, B=@n41.XYZ, Radius=@n42) {hardnessPercent=40}
   )
   B=Add(A=Multiply(A=@n56, B=@n37.R), B=Multiply(A=@n64, B=@n52.Result))
  )
 ]
}
