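# Translucent, unlit, two-sided material for an icy fog cloud surround, built on panning storm noise and normal textures.
# Color fakes lighting by dotting the offset sun/moon vector with the normal map, adding time-of-day highlight and shadow tints, then modulating by noise.
# Opacity combines noise, a height fade above WaterHeight, view-angle fade and UV1 sphere masks driven partly by Reveal.
# A quality switch provides time-of-day-varying or fixed highlight power.
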Material /Figment_S06/Effects/Fort_Effects/Materials/Smoke/M_Ice_Fog_Cloud_Surround
graph {
 properties {blendMode=BLEND_Translucent, pixelDepthOffsetMode=PDOM_Legacy, shadingModel=MSM_Unlit, twoSided=true, usedShadingModels=Unlit}
 nodes [
  Constant() {r=1}
  MaterialFunctionCall() {materialFunction=&DebugTimeSine}
  n1 = ComponentMask(None=Subtract(A=WorldPosition().XYZ, B=ScalarParameter() {defaultValue=-5000, parameterName=WaterHeight})) {b=true}
  n2 = ScalarParameter() {defaultValue=20000, parameterName=HeightFadeGradientSize}
  n3 = Panner(Coordinate=TextureCoordinate() {uTiling=2}, Time=Multiply(A=Time(), B=ScalarParameter() {defaultValue=0.1, parameterName=PandSPeed})) {speedX=0.1, speedY=-0.05}
  n4 = TextureSample(UVs=@n3) {texture=/Game/VisualThreatManager/StormVisuals/Textures/2ndInfectedMapLinear}
  n5 = LinearInterpolate(
   A=ScalarParameter() {defaultValue=1.4, parameterName=OpacityMin}
   B=ScalarParameter() {defaultValue=1.8, parameterName=OpacityMax}
   Alpha=@n4.G
  ) {constA=1.4}
  n6 = MaterialFunctionCall("FadeSensitivity (optional)"=ScalarParameter() {defaultValue=1, parameterName=AngleFade}) {materialFunction=&FadeBasedOnViewAngle}
  n7 = Constant() {r=0.5}
  n8 = SphereMask(A=ComponentMask(None=TextureCoordinate() {coordinateIndex=1}) {r=true}, B=@n7, Radius=@n7) {attenuationRadius=0.5, hardnessPercent=50}
  n9 = SphereMask(
   A=ComponentMask(None=TextureCoordinate() {coordinateIndex=1}) {g=true}
   B=LinearInterpolate(Alpha=ScalarParameter() {parameterName=Reveal}) {constA=1, constB=1.25}
  ) {attenuationRadius=1, hardnessPercent=80}
  Multiply(A=Clamp(None=Divide(A=@n1, B=@n2)), B=Multiply(A=Multiply(A=@n5, B=@n6.Result), B=Multiply(A=@n8, B=@n9)))
  n10 = CollectionParameter() {
   collection=&FortniteMaterialParameters
   parameterName=FortniteActiveTimeOfDay
  }
  n11 = CollectionParameter() {
   collection=&FortniteMaterialParameters
   parameterName=SunAndMoonModelDirectionalVector
  }
  n12 = Add(A=ComponentMask(None=@n11) {b=true, g=true, r=true}, B=VectorParameter() {parameterName=SunVectorOffset, defaultValue=<rgba 0.4 0 0 1>}.RGB)
  n13 = Multiply(
   A=VectorParameter() {parameterName=NormalTransform, defaultValue=<rgba -1 1 1 0>}.RGB
   B=Multiply(
    A=TextureSample(UVs=@n3) {samplerType=SAMPLERTYPE_Normal, texture=/Game/VisualThreatManager/StormVisuals/Textures/2ndInfectedMapNormal}.RGB
    B=VectorParameter() {parameterName=NormalMult, defaultValue=<rgba 0.5 0.5 1 0>}.RGB
   )
  )
  n14 = Clamp(None=DotProduct(A=Normalize(VectorInput=@n12), B=Transform(None=@n13)))
  n15 = Multiply(
   A=LinearInterpolate(Alpha=ComponentMask(None=@n10) {a=true}) {constA=1, constB=0.5}
   B=Multiply(A=VectorParameter() {parameterName=HighlightColor, defaultValue=<rgba 0.8125 0.9658 1 0>}.RGB, B=Power(Base=@n14)) {constB=0}
  )
  n16 = MaterialFunctionCall(
   Color_Evening=VectorParameter() {parameterName=EveningTint, defaultValue=<rgba 1 0.4864 0.2344 1>}.RGB
   Color_Night=VectorParameter() {parameterName=NightTint, defaultValue=<rgba 0.3311 0.2457 0.7 1>}.RGB
  ) {materialFunction=&Color_Time_of_Day}
  n17 = VectorParameter() {parameterName=ShadowMorning, defaultValue=<rgba 0.9 0.5487 0.4078 0>}
  n18 = VectorParameter() {parameterName=ShadowDay, defaultValue=<rgba 0.4531 0.5749 1 0>}
  n19 = VectorParameter() {parameterName=ShadowEvening, defaultValue=<rgba 0.4 0.1457 0.1537 1>}
  n20 = VectorParameter() {parameterName=ShadowNight, defaultValue=<rgba 0.3649 0.3478 0.7 0>}
  Multiply(
   A=Add(
    A=Multiply(A=@n15, B=@n16.Result)
    B=MaterialFunctionCall(Color_Morning=@n17.RGB, Color_Day=@n18.RGB, Color_Evening=@n19.RGB, Color_Night=@n20.RGB) {materialFunction=&Color_Time_of_Day}.Result
   )
   B=LinearInterpolate(Alpha=@n4.G) {constA=0.4, constB=0.7}
  )
  n21 = Cosine(None=Power(
    Base=CollectionParameter() {collection=&FortniteMaterialParameters, parameterName=TimeOfDay_0-1}
    Exp=ScalarParameter() {defaultValue=2, parameterName=TODPower}
   ))
  QualitySwitch(
   Default=LinearInterpolate(Alpha=ConstantBiasScale(None=@n21)) {constA=6, constB=1.8}
   Low=ScalarParameter() {defaultValue=2, parameterName=Spec Highlight Power}
  )
 ]
}
