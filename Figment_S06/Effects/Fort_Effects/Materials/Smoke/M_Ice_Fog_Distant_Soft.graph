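# Translucent unlit two-sided material for large, soft, distant ice fog built from camera-facing cards.
# World position offset math billboards each card around the object position, using UV1 to spread cards and shrink them toward the edges.
# A panning storm swirl mask drives opacity within a soft radial mask, faded by depth and height above WaterHeight.
# Color combines time-of-day tints, shadow color and a swirl-normal highlight from the sun/moon vector.
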Material /Figment_S06/Effects/Fort_Effects/Materials/Smoke/M_Ice_Fog_Distant_Soft
graph {
 properties {blendMode=BLEND_Translucent, pixelDepthOffsetMode=PDOM_Legacy, shadingModel=MSM_Unlit, twoSided=true, usedShadingModels=Unlit}
 nodes [
  n1 = TextureCoordinate() {coordinateIndex=1}
  n2 = Sine(None=Add(A=ScalarParameter() {defaultValue=0.3, parameterName=XSpreadSineOffset}, B=ComponentMask(None=@n1) {r=true})) {period=0.33}
  n3 = ComponentMask(None=@n1) {g=true, r=true}
  n4 = Multiply(A=AppendVector(A=Multiply(A=@n2) {constB=0.5}, B=Subtract(A=@n3) {constB=0.5})) {constB=-1}
  n5 = MaterialFunctionCall(Float2=Subtract(A=TextureCoordinate()) {constB=0.5}) {materialFunction=&BreakOutFloat2Components}
  n6 = Subtract(A=ObjectPositionWS(), B=CameraPositionWS())
  n7 = CrossProduct(A=@n6, B=Constant3Vector() {constant=<rgba 0 0 1 0>})
  n8 = Normalize(VectorInput=@n7)
  n9 = ScalarParameter() {defaultValue=300000, parameterName=Scale}
  n10 = SphereMask(A=TextureCoordinate() {coordinateIndex=1}, B=Constant() {r=0.5}) {attenuationRadius=0.85, hardnessPercent=0}
  n11 = MaterialFunctionCall(Float2=Multiply(A=Multiply(A=@n9, B=@n10), B=Constant2Vector() {g=0.75, r=1})) {materialFunction=&BreakOutFloat2Components}
  n12 = Normalize(VectorInput=CrossProduct(A=@n6, B=@n7))
  n13 = Add(A=Multiply(A=Multiply(A=@n5.R, B=@n8), B=@n11.R), B=Multiply(A=Multiply(A=@n5.G, B=@n12), B=@n11.G))
  n14 = WorldPosition()
  Add(
   A=Multiply(
    A=Transform(None=@n4) {transformSourceType=TRANSFORMSOURCE_Local}
    B=VectorParameter() {parameterName=Spread, defaultValue=<rgba 900 1400 500 0>}.RGB
   )
   B=Add(A=Subtract(A=@n13, B=@n14.XYZ), B=ObjectPositionWS())
  )
  n15 = ComponentMask(None=Subtract(A=WorldPosition().XYZ, B=ScalarParameter() {defaultValue=-5000, parameterName=WaterHeight})) {b=true}
  n16 = TextureCoordinate() {coordinateIndex=1}
  n17 = Subtract(A=@n16) {constB=0.5}
  n18 = Panner(
   Coordinate=Add(A=TextureCoordinate() {uTiling=0.55, vTiling=0.55}, B=@n16)
   Speed=Multiply(A=@n17, B=ScalarParameter() {defaultValue=0.03, parameterName=PanSpeed}) {constB=0.3}
  )
  n19 = Multiply(
   A=SphereMask(A=TextureCoordinate(), B=Constant() {r=0.5}) {attenuationRadius=0.49, hardnessPercent=0}
   B=LinearInterpolate(
    A=ScalarParameter() {defaultValue=0.6, parameterName=DetailMinOpacity}
    Alpha=TextureSample(UVs=@n18) {samplerType=SAMPLERTYPE_LinearGrayscale, texture=/Figment_S06/VisualThreatManager/StormVisuals/LargeClouds/T_StormSwirlMask}.G
   )
  )
  n20 = LinearInterpolate(Alpha=@n19) {constA=0.45, constB=0.05}
  n21 = Multiply(A=ScalarParameter() {defaultValue=90000, parameterName=HeightFadeGradientSize}, B=@n20)
  n22 = ScalarParameter() {defaultValue=0.3, parameterName=OpacityMult}
  n23 = Multiply(A=@n20, B=@n9) {constB=0.5}
  Multiply(A=Clamp(None=Divide(A=@n15, B=@n21)), B=Multiply(A=Multiply(A=@n19, B=@n22), B=DepthFade(FadeDistance=@n23)))
  Multiply(A=@n17, B=Constant2Vector() {g=1, r=1})
  n24 = VectorParameter() {parameterName=MorningTint, defaultValue=<rgba 3 1.003 0.366 1>}
  n25 = VectorParameter() {parameterName=EveningTint, defaultValue=<rgba 0.7 0.1569 0.0308 1>}
  n26 = VectorParameter() {parameterName=NightTint, defaultValue=<rgba 0.1663 0.1896 0.5 1>}
  n27 = CollectionParameter() {
   collection=&FortniteMaterialParameters
   parameterName=SunAndMoonModelDirectionalVector
  }
  n28 = ComponentMask(None=@n27) {b=true, g=true, r=true}
  n29 = VectorParameter() {parameterName=SunVectorOffset, defaultValue=<rgba 0.4 0 0 1>}
  n30 = TextureSample(UVs=@n18) {samplerType=SAMPLERTYPE_Normal, texture=/Game/VisualThreatManager/StormVisuals/LargeClouds/T_StormSwirlNormal}
  n31 = VectorParameter() {parameterName=SwirlNormalMult, defaultValue=<rgba 0.6 0.6 1 1>}
  n32 = DotProduct(A=Normalize(VectorInput=Add(A=@n28, B=@n29.RGB)), B=Transform(None=Multiply(A=@n30.RGB, B=@n31.RGB)))
  n33 = Multiply(
   A=Power(Base=Clamp(None=@n32), Exp=ScalarParameter() {defaultValue=3, parameterName=Spec Highlight Power}) {constExponent=20}
   B=VectorParameter() {parameterName=HighlightColor, defaultValue=<rgba 1.625 1.9315 2 0>}.RGB
  ) {constB=0}
  n34 = Power(
   Base=CollectionParameter() {collection=&FortniteMaterialParameters, parameterName=TimeOfDay_0-1}
   Exp=ScalarParameter() {defaultValue=3, parameterName=TODPower}
  )
  n35 = LinearInterpolate(Alpha=ConstantBiasScale(None=Cosine(None=@n34))) {constA=0.4}
  Multiply(
   A=LinearInterpolate(A=MaterialFunctionCall(Color_Morning=@n24.RGB, Color_Evening=@n25.RGB, Color_Night=@n26.RGB) {materialFunction=&Color_Time_of_Day}.Result) {constAlpha=0.2}
   B=Add(A=VectorParameter() {parameterName=ShadowColor, defaultValue=<rgba 0.2905 0.3594 0.6 0>}.RGB, B=Multiply(A=@n33, B=@n35))
  )
 ]
}
